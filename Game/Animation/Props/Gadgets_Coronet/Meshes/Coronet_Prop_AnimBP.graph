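# Animation Blueprint for the Coronet gadget crown prop, playing DefaultSlot over the reference pose and setting crown elements by the player's Royal Royale count.
# On initialization it watches the pawn's TotalRoyalRoyales and UsePremiumColors cosmetic stats, then creates a dynamic material.
# Curve-table thresholds select which "Hide Element" parameters to set, leaving one element per tier or none below the lowest.
# This selection runs only at initialization, and UsePremiumColors is stored but unused here.
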AnimBlueprint /Game/Animation/Props/Gadgets_Coronet/Meshes/Coronet_Prop_AnimBP
target_skeleton=&/Coronet_Gadget_Skeleton
graphs [
 AnimGraph {nodes [Output Pose(Result=Slot 'DefaultSlot'(Source=Local Space Ref Pose())) {layerGroup=None}]}
 EventGraph {nodes [
   Event BlueprintInitializeAnimation()
   > n1 = Sequence!(execute) -> (then_0, then_1, then_2:Exec)
   > n2 = Cast To FortPlayerPawn(Object=TryGetPawnOwner())
   > n3 = WatchCosmeticStat!(
    execute:Exec
    Pawn:Fort Player Pawn Object Reference=@n2
    StatTag:Fort Stat Manager Tag Structure=(TagName="StatManager.TotalRoyalRoyales")
    DefaultValue:int=0
    bBroadcastInitialValue:bool=true
    InitialValueBroadcastDelay:float=-1
    ChangeValueBroadcastDelay:float=-1
    bConsumeWaitingNotifications:bool=true
   ) -> (then:Exec, OnStatChanged:Exec, StatName:Name, StatValue:int) {is=title K2Node_AsyncAction}
   Set RoyalRoyalesAchieved(execute=[@n3.OnStatChanged, @n3.then], RoyalRoyalesAchieved:int=@n3.StatValue) -> (Output_Get:int=0)
   > n4 = WatchCosmeticStat(
    Pawn=@n2
    StatTag=(TagName="StatManager.UsePremiumColors")
    DefaultValue=0
    bBroadcastInitialValue=true
    InitialValueBroadcastDelay=-1
    ChangeValueBroadcastDelay=-1
    bConsumeWaitingNotifications=true
   )
   Set UsePremiumColors(execute=[@n4.OnStatChanged, @n4.then], UsePremiumColors:bool="To Boolean (Integer)"!(InInt:int=@n4.StatValue) -> (ReturnValue:bool)) -> (Output_Get:bool)
   n5 = Get Dynamic Material Instance(execute=@n1.then_1, SkeletalMeshComponent=GetOwningComponent())
   > Set CrownMaterial(CrownMaterial:Material Instance Dynamic Object Reference=@n5) -> (Output_Get:Material Instance Dynamic Object Reference)
   n6 = Get RoyalRoyaleThreshold() -> (RoyalRoyaleThreshold:Scalable Float Structure)
   n7 = Branch(
    execute=@n1.then_2
    Condition="float >= float"(
     A:int=Get RoyalRoyalesAchieved() -> (RoyalRoyalesAchieved:int=0)
     B=GetValueAtLevel!(Input:"Scalable Float Structure (by ref)"=@n6, Level:float=3) -> (ReturnValue:float)
    )
   )
   > SetScalarParameterValue(self=Get CrownMaterial() -> (CrownMaterial:Material Instance Dynamic Object Reference), ParameterName=Hide Element 01, Value=1)
   > SetScalarParameterValue(self=Get CrownMaterial(), ParameterName=Hide Element 02, Value=1)
   n8 = Branch(
    execute=@n7.else
    Condition="float >= float"(A=Get RoyalRoyalesAchieved() -> (RoyalRoyalesAchieved=0), B=GetValueAtLevel(Input=@n6, Level=2))
   )
   > SetScalarParameterValue(self=Get CrownMaterial(), ParameterName=Hide Element 01, Value=1)
   > SetScalarParameterValue(self=Get CrownMaterial(), ParameterName=Hide Element 03, Value=1)
   n9 = Branch(
    execute=@n8.else
    Condition="float >= float"(A=Get RoyalRoyalesAchieved() -> (RoyalRoyalesAchieved=0), B=GetValueAtLevel(Input=@n6, Level=1))
   )
   > SetScalarParameterValue(self=Get CrownMaterial(), ParameterName=Hide Element 02, Value=1)
   > SetScalarParameterValue(self=Get CrownMaterial(), ParameterName=Hide Element 03, Value=1)
   SetScalarParameterValue(execute=@n9.else, self=Get CrownMaterial(), ParameterName=Hide Element 01, Value=1)
   > SetScalarParameterValue(self=Get CrownMaterial(), ParameterName=Hide Element 02, Value=1)
   > SetScalarParameterValue(self=Get CrownMaterial(), ParameterName=Hide Element 03, Value=1)
  ]}
]
variables [
 CrownMaterial {type=MaterialInstanceDynamic}
 RoyalRoyalesAchieved {type=integer}
 RoyalRoyaleThreshold {
  type=ScalableFloat
  default {
   registryType=None
   value=1
   curve {curveTable=/Game/Athena/Items/Cosmetics/DataTables/AthenaCosmeticsData, rowName=Default.Emote.Coronet.RoyalRoyaleThreshold}
  }
 }
 UsePremiumColors {type=boolean}
]
